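# Niagara system for the knockback weapon's overheat effect, combining barrel smoke with a glow tinted by User.Color, red-orange by default.
# Smoke_Barrel emits rising flipbook sprites that shift from hot red to white, expand and fade.
# BigGlow layers three sprite renderers using derived size and alpha attributes, fading over the emitter loop, while Decal bursts one longer-lived tinted sprite.
# Bolt, light, spark and two renderer-less emitters are disabled.
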NiagaraSystem /Skyfire/Gameplay/KnockbackWeapon/FX/Overheat/NS_Knockback_Overheat
dependencies {
 usedDataInterfaces [&NiagaraDataInterfaceCurve, &NiagaraDataInterfaceColorCurve, &NiagaraDataInterfaceVectorCurve]
 usedDynamicInputs []
 usedModules [&SystemState]
}
emitters {
 BigGlow {
  emitterUpdateScript [
   &EmitterState()
   &SpawnRate(SpawnRate=6, Use Spawn Probability=false)
  ]
  particleSpawnScript [&InitializeParticle(
    Color=$User.Color
    Lifetime Max=0.35
    Lifetime Min=0.35
    Lifetime Mode=::NewEnumerator1
    Sprite Rotation Mode=::NewEnumerator2
    Sprite Size Mode=::NewEnumerator1
    Uniform Sprite Size Max=22
    Uniform Sprite Size Min=18
    UsePositionOffset=false
    Write Material Random=true
   ) {module=InitializeParticle001}]
  particleUpdateScript [
   &ParticleState()
   &ScaleColor(
    Scale Alpha=&FloatFromCurve(
     FloatCurve=DataInterface() {propertyValues {Curve {keys [
         <at 0 0 RCIM_Cubic>
         {arriveTangent=1.4286, interpMode=RCIM_Cubic, leaveTangent=1.4286, time=0.1, value=1}
         {arriveTangent=-1.1111, interpMode=RCIM_Cubic, leaveTangent=-1.1111, time=0.7, value=1}
         <at 1 0>
        ]}}}
     Scale Curve=0.05
    )
    ScaleA=true
    ScaleRGB=false
   )
   &CameraOffset(Camera Offset Amount=20)
   &FN_Sprite_Attributes(
    Color=&MakeLinearColorFromVectorAndFloat(
     "Vector (RGB)"=&MakeVectorFromLinearColorRGB(Color=$Particles.Color)
     "Float (Alpha)"=&Add_Float(
      A=0.02
      B=&Multiply_Float(
       A=&MakeFloatFromLinearColor(LinearColor=$Particles.Color, Channel:ENiagaraLinearColor_Channels=::NewEnumerator3)
       B=0.02
      )
     )
    )
    Create Color Attribute=true
    Create SpriteSize Attribute=true
    SpriteSize=/Niagara/DynamicInputs/Add/Add_Vector2(A:Vector2f=<35 35>, B:Vector2f=&Multiply_Vector2(A=$Particles.SpriteSize, B=<0.2 0.2>))
    Write to Module Namespace=true
   )
   &FN_Sprite_Attributes(
    Color=&MakeLinearColorFromVectorAndFloat(
     "Vector (RGB)"=&MakeVectorFromLinearColorRGB(Color=$Particles.Color)
     "Float (Alpha)"=&Add_Float(
      A=0.15
      B=&Multiply_Float(
       A=&MakeFloatFromLinearColor(LinearColor=$Particles.Color, Channel:ENiagaraLinearColor_Channels=::NewEnumerator3)
       B=0.02
      )
     )
    )
    Create Color Attribute=true
    Create SpriteSize Attribute=true
    SpriteSize=&Add_Vector2(A:Vector2f=<10 10>, B:Vector2f=&Multiply_Vector2(A=$Particles.SpriteSize, B=<0.2 0.2>))
    Write to Module Namespace=true
   ) {module=FN_Sprite_Attributes001}
   &ScaleColor(
    Color Value To Scale=$Particles.Color
    Scale Alpha=&FloatFromCurve(
     FloatCurve=DataInterface() {propertyValues {Curve {keys [
         {tangentMode=RCTM_User, time=0, value=1}
         {arriveTangent=-0.0405, interpMode=RCIM_Cubic, leaveTangent=-0.0405, tangentMode=RCTM_User, time=0.7009, value=1}
         {arriveTangent=-8.0089, interpMode=RCIM_Cubic, leaveTangent=-8.0089, tangentMode=RCTM_User, time=1, value=0}
        ]}}}
     CurveIndex=$Emitter.NormalizedLoopAge
    )
    ScaleA=true
    ScaleRGB=false
   ) {module=ScaleColor001}
  ]
  rendererClasses [&NiagaraSpriteRendererProperties]
  renderers [
   {rendererClass=&NiagaraSpriteRendererProperties}
   {rendererClass=&NiagaraSpriteRendererProperties, rendererIndex=1}
   {rendererClass=&NiagaraSpriteRendererProperties, rendererIndex=2}
  ]
 }
 BoltsLarge {
  bEnabled=false
  emitterUpdateScript [
   &EmitterState(Life Cycle Mode=::NewEnumerator1, Loop Behavior=::NewEnumerator1, Loop Duration=2.5, UseLoopDelay=true)
   &SpawnRate(SpawnRate=25, Use Spawn Probability=false)
   &SpawnBurst_Instantaneous(Spawn Count=&RandomRangeInt(Minimum=1, Maximum=1), Use Spawn Probability=false)
  ]
  particleSpawnScript [
   /SetVariables(Particles.MeshOrientation=&MakeQuatFromAxisAngle(
     Axis=&RandomRangeVector(Minimum=<0 0 0>, Maximum=<0 1 0>, Evaluation Type=::NewEnumerator1)
     AngleInDegrees=90
    )) {isSetParameters=true}
   /Niagara/Modules/Spawn/Initialization/InitializeParticle(
    Lifetime=&RandomRangeFloat(Minimum=0.05, Maximum=0.08, Evaluation Type=::NewEnumerator1)
    Mesh Scale=<0.06 0.02 0.03>
    Module.Write Color=true
    Module.Write Lifetime=true
    Module.Write Mass=true
    Module.Write Material Random=true
    Module.Write Position=false
    Module.Write Scale=true
    Module.Write SpriteRotation=false
    Module.Write SpriteSize=false
   )
   &DynamicMaterialParameters(
    Index 0 Param 2=&RandomRangeFloat(Minimum=0, Maximum=1, Evaluation Type=::NewEnumerator1)
    Index 0 Param 3=&RandomRangeFloat(Minimum=1, Maximum=3, Evaluation Type=::NewEnumerator1)
    Index 0 Param 4=1
    Module.Param0WriteEnabled=false
    Module.Param1WriteEnabled=true
    Module.Param2WriteEnabled=true
    Module.Param3WriteEnabled=true
   ) {module=DynamicMaterialParameters001}
   /Niagara/Modules/Spawn/Location/SphereLocation(MaskSpawnGroup=false, Module.Override Randomness=false, Offset=<0 5 0>, Sphere Radius=3, Surface Only=true)
   &CylinderLocation(
    Cylinder Height=37.1708
    Cylinder Radius=11
    MaskSpawnGroup=false
    Module.Enable Lathe Profile=false
    Module.Override Randomness=false
    Module.Surface Only=true
    Offset=<0 43 20>
    Surface Only Band Thickness=0
    Use Endcaps In Surface Only Mode=false
   )
   &Color(Color=&MakeLinearColorFromVectorAndFloat("Vector (RGB)"=<0 4 3.8888>, "Float (Alpha)"=&ClampFloat(Float=1)))
  ]
  particleUpdateScript [
   /SetVariables(Particles.LightExponent=3) {isSetParameters=true}
   /SetVariables(Particles.LightColor:Vector3f=<1 1 1>) {isSetParameters=true}
   /SetVariables(Particles.LightRadius=900) {isSetParameters=true}
   &ParticleState()
   &DynamicMaterialParameters(
    Index 0 Param 1=0.08
    Module.Param0WriteEnabled=true
    Module.Param1WriteEnabled=false
    Module.Param2WriteEnabled=false
    Module.Param3WriteEnabled=false
   )
  ]
  rendererClasses [&NiagaraLightRendererProperties]
  renderers [{rendererClass=&NiagaraLightRendererProperties, rendererIndex=0}]
 }
 Decal {
  emitterUpdateScript [
   &EmitterState()
   &SpawnRate(SpawnRate=2, Use Spawn Probability=false) {enabled=false}
   &SpawnBurst_Instantaneous(Spawn Count=1, Use Spawn Probability=false)
  ]
  particleSpawnScript [/Niagara/Modules/Spawn/Initialization/V2/InitializeParticle(
    Color=$User.Color
    Lifetime Max=2.5
    Lifetime Min=2.5
    Lifetime Mode=::NewEnumerator1
    Sprite Rotation Mode=::NewEnumerator2
    Sprite Size=<25 50>
    Sprite Size Mode=::NewEnumerator4
    UsePositionOffset=false
    Write Material Random=true
   ) {module=InitializeParticle001}]
  particleUpdateScript [
   &ParticleState()
   &ScaleColor(
    Scale Alpha=&FloatFromCurve(
     FloatCurve=DataInterface() {propertyValues {Curve {keys [
         <at 0 0 RCIM_Cubic>
         {arriveTangent=1.4286, interpMode=RCIM_Cubic, leaveTangent=1.4286, time=0.1, value=1}
         {arriveTangent=-1.1111, interpMode=RCIM_Cubic, leaveTangent=-1.1111, time=0.7, value=1}
         <at 1 0>
        ]}}}
     Scale Curve=0.25
    )
    ScaleA=true
    ScaleRGB=false
   )
   &CameraOffset(Camera Offset Amount=20)
   &ScaleColor(
    Color Value To Scale=$Particles.Color
    Scale Alpha=&FloatFromCurve(
     FloatCurve=DataInterface() {propertyValues {Curve {keys [
         {tangentMode=RCTM_User, time=0, value=1}
         {arriveTangent=-0.0405, interpMode=RCIM_Cubic, leaveTangent=-0.0405, tangentMode=RCTM_User, time=0.2, value=1}
         {arriveTangent=-3.4457, interpMode=RCIM_Cubic, leaveTangent=-3.4457, tangentMode=RCTM_User, time=1, value=0}
        ]}}}
     CurveIndex=$Emitter.NormalizedLoopAge
    )
    ScaleA=true
    ScaleRGB=false
   ) {enabled=false, module=ScaleColor001}
   &FN_Sprite_Attributes(
    Color=&MakeLinearColorFromVectorAndFloat(
     "Vector (RGB)"=&MakeVectorFromLinearColorRGB(Color=$Particles.Color)
     "Float (Alpha)"=&Add_Float(
      A=0.15
      B=&Multiply_Float(
       A=&MakeFloatFromLinearColor(LinearColor=$Particles.Color, Channel:ENiagaraLinearColor_Channels=::NewEnumerator3)
       B=0.02
      )
     )
    )
    Create Color Attribute=true
    Create SpriteSize Attribute=true
    SpriteSize=&Add_Vector2(A:Vector2f=<10 10>, B:Vector2f=&Multiply_Vector2(A=$Particles.SpriteSize, B=<0.2 0.2>))
    Write to Module Namespace=true
   ) {module=FN_Sprite_Attributes001}
  ]
  rendererClasses [&NiagaraSpriteRendererProperties]
  renderers [{rendererClass=&NiagaraSpriteRendererProperties}]
 }
 Light {
  bEnabled=false
  emitterUpdateScript [
   &EmitterState(Life Cycle Mode=::NewEnumerator1, Loop Behavior=::NewEnumerator1, Loop Duration=1, UseLoopDelay=true)
   &SpawnBurst_Instantaneous(Spawn Count=2, Use Spawn Probability=false)
  ]
  particleSpawnScript [
   /Niagara/Modules/Spawn/Initialization/InitializeParticle(
    Lifetime=2
    Module.Write Color=true
    Module.Write Lifetime=true
    Module.Write Mass=true
    Module.Write Material Random=true
    Module.Write Position=false
    Module.Write Scale=false
    Module.Write SpriteRotation=false
    Module.Write SpriteSize=true
    Position=&RandomRangeVector(Minimum=<-1 0 25>, Maximum=<35 0 25>, Evaluation Type=::NewEnumerator1)
    Sprite Size=<1 1>
   )
   &Color(Color=&MakeLinearColorFromVectorAndFloat("Vector (RGB)"=<300 300 300>, "Float (Alpha)"=&ClampFloat(Float=1)))
  ]
  particleUpdateScript [
   /SetVariables(Particles.LightExponent=16) {isSetParameters=true}
   /SetVariables(Particles.LightColor:Vector3f=<1 1 1>) {isSetParameters=true}
   /SetVariables(Particles.LightRadius=10) {isSetParameters=true}
   &ParticleState()
   &ScaleColor(
    Scale RGB=&VectorFromCurve(VectorCurve=DataInterface() {propertyValues {
       XCurve {keys [<at 0.25 0 RCIM_Cubic>,<at 0.7951 0 RCIM_Cubic>,<at 1 0 RCIM_Cubic>]}
       YCurve {keys [
         <at 0.25 0.049 RCIM_Cubic>
         {arriveTangent=-0.0653, interpMode=RCIM_Cubic, leaveTangent=-0.0653, time=0.7951, value=0.2}
         <at 1 0 RCIM_Cubic>
        ]}
       ZCurve {keys [<at 0.25 1 RCIM_Cubic>, {arriveTangent=-1.3333, interpMode=RCIM_Cubic, leaveTangent=-1.3333, time=0.7951, value=0.03}, <at 1 0 RCIM_Cubic>]}
      }})
    ScaleA=true
    ScaleRGB=true
   )
  ]
  rendererClasses [&NiagaraSpriteRendererProperties, &NiagaraLightRendererProperties]
  renderers [
   {rendererClass=&NiagaraSpriteRendererProperties}
   {rendererClass=&NiagaraLightRendererProperties, rendererIndex=1}
  ]
 }
 Particle_Emitter {
  bEnabled=false
  emitterUpdateScript [
   &EmitterState(Life Cycle Mode=::NewEnumerator1, Loop Behavior=::NewEnumerator1, Loop Delay=0.1, Loop Duration=1, UseLoopDelay=true)
   &SpawnBurst_Instantaneous(Spawn Count=1, Use Spawn Probability=false)
  ]
  particleSpawnScript [
   &InitializeParticle(
    Mesh Scale=<2 2 2>
    Module.Write Color=true
    Module.Write Lifetime=true
    Module.Write Mass=true
    Module.Write Material Random=true
    Module.Write Position=false
    Module.Write Scale=true
    Module.Write SpriteRotation=false
    Module.Write SpriteSize=false
    Position=<0 76 20>
   )
   &Color(Color=&MakeLinearColorFromVectorAndFloat("Vector (RGB)"=<0 20088.2441 25000>, "Float (Alpha)"=&ClampFloat(Float=0.05)))
  ]
  particleUpdateScript [
   &ParticleState()
   &DynamicMaterialParameters(
    Index 0 Param 1=&FloatFromCurve(FloatCurve=DataInterface() {propertyValues {Curve {keys [{arriveTangent=5.3858, interpMode=RCIM_Cubic, time=0, value=0}, <at 0.974 2 RCIM_Cubic>]}}})
    Index 0 Param 2=&FloatFromCurve(FloatCurve=DataInterface() {propertyValues {Curve {keys [<at 0 15>,<at 1 5>]}}})
    Index 0 Param 3=&FloatFromCurve(FloatCurve=DataInterface() {propertyValues {Curve {keys [<at 0 -0.8>,<at 1 -0.8>]}}})
    Index 0 Param 4=&FloatFromCurve(FloatCurve=DataInterface() {propertyValues {Curve {keys [<at 0 1>,<at 1 1>]}}})
    Module.Param0WriteEnabled=true
    Module.Param1WriteEnabled=true
    Module.Param2WriteEnabled=true
    Module.Param3WriteEnabled=true
   )
   &ScaleColor(
    Scale Alpha=&FloatFromCurve(FloatCurve=DataInterface() {propertyValues {Curve {keys [
         {arriveTangent=-3.0005, interpMode=RCIM_Cubic, time=0, value=1}
         {arriveTangent=-2.6944, interpMode=RCIM_Cubic, leaveTangent=-2.6944, time=0.1011, value=0.1956}
         {arriveTangent=-0.2176, interpMode=RCIM_Cubic, leaveTangent=-0.2176, time=0.351, value=0.0542}
         <at 1 0 RCIM_Cubic>
        ]}}})
    Scale RGB=&VectorFromCurve(VectorCurve=DataInterface() {propertyValues {XCurve {keys [<at 0 2>,<at 1 0>]}, YCurve {keys [<at 0 2>,<at 1 1>]}, ZCurve {keys [<at 0 2>,<at 1 0.25>]}}})
    ScaleA=true
    ScaleRGB=true
   )
  ]
  rendererClasses []
  renderers []
 }
 Particle_Emitter001 {
  bEnabled=false
  emitterUpdateScript [
   &EmitterState(Life Cycle Mode=::NewEnumerator1, Loop Behavior=::NewEnumerator1, Loop Duration=1, UseLoopDelay=true)
   &SpawnBurst_Instantaneous(Spawn Count=1, Use Spawn Probability=false)
  ]
  particleSpawnScript [
   &InitializeParticle(
    Lifetime=0.12
    Mesh Scale=<2 2 2>
    Module.Write Color=true
    Module.Write Lifetime=true
    Module.Write Mass=true
    Module.Write Material Random=true
    Module.Write Position=false
    Module.Write Scale=true
    Module.Write SpriteRotation=false
    Module.Write SpriteSize=false
    Position=<0 76 20>
   )
   &Color(Color=&MakeLinearColorFromVectorAndFloat("Vector (RGB)"=<0.01 0.2 0.3>, "Float (Alpha)"=&ClampFloat(Float=0)))
  ]
  particleUpdateScript [&ParticleState()]
  rendererClasses []
  renderers []
 }
 Smoke_Barrel {
  emitterUpdateScript [
   &EmitterState()
   &SpawnRate(
    SpawnRate=&Multiply_Float(
     A=10
     B=&FloatFromCurve(FloatCurve=DataInterface() {propertyValues {Curve {keys [<at 1 1>,<at 2 0.7>,<at 3 0>]}}}, CurveIndex=$Emitter.NormalizedLoopAge)
    )
    Use Spawn Probability=false
   )
   &SpawnBurst_Instantaneous(Spawn Count=2, Use Spawn Probability=false)
  ]
  particleSpawnScript [
   /SetVariables(Particles.SpriteRotationRate=&RandomRangeFloat(Minimum=-0.2, Maximum=0.2, Evaluation Type=::NewEnumerator1)) {isSetParameters=true}
   &InitializeParticle(
    Lifetime=&RandomRangeFloat(Minimum=0.5, Maximum=1, Evaluation Type=::NewEnumerator1)
    Module.Write Color=true
    Module.Write Lifetime=true
    Module.Write Mass=true
    Module.Write Material Random=true
    Module.Write Position=false
    Module.Write Scale=false
    Module.Write SpriteRotation=true
    Module.Write SpriteSize=true
    Position=<0 68 20>
    Sprite Rotation=&NormalizedAngleToDegrees(NormalizedAngle=&RandomRangeFloat(Minimum=0, Maximum=1, Evaluation Type=::NewEnumerator1))
    Sprite Size=&Vector2DFromFloat(Value=&RandomRangeFloat(Minimum=10, Maximum=14, Evaluation Type=::NewEnumerator1))
   )
   /Niagara/Modules/Spawn/Location/V2/SphereLocation(
    MaskSpawnGroup=false
    Offset=<0 0 -2>
    Override Seed=false
    Sphere Radius=3
    Surface Only=false
    UseNonUniformScale=false
    UseOffset=true
    UseSphereOrientation=false
   )
   &AddVelocity(Velocity=&RandomRangeVector(Minimum=<-10 10 28>, Maximum=<10 15 35>, Evaluation Type=::NewEnumerator1))
   &Color(Color=&MakeLinearColorFromVectorAndFloat(
     "Vector (RGB)"=<1 1 1>
     "Float (Alpha)"=&ClampFloat(Float=&RandomRangeFloat(Minimum=0.05, Maximum=0.08, Evaluation Type=::NewEnumerator1))
    ))
  ]
  particleUpdateScript [
   &ParticleState()
   &Drag(Drag=0.8, Use Linear Drag=true, Use Rotational Drag=false)
   &SolveForcesAndVelocity(Clamp Velocity=false, Limit Acceleration=false)
   &Color(Color=&ColorFromCurve(ColorCurve=DataInterface() {propertyValues {
       AlphaCurve {keys [<at 0 1>,<at 1 1 RCIM_Cubic>]}
       BlueCurve {keys [<at 0 0.1 RCIM_Cubic>,<at 1 1 RCIM_Cubic>]}
       GreenCurve {keys [<at 0 0.2 RCIM_Cubic>,<at 1 1 RCIM_Cubic>]}
       RedCurve {keys [<at 0 1.5 RCIM_Cubic>,<at 1 1 RCIM_Cubic>]}
      }})) {module=Color001}
   /Niagara/Modules/Update/SubUV/SubUVAnimation(Number Of Frames=32)
   &ScaleColor(
    Color Value To Scale=$Particles.Color
    Scale Alpha=&FloatFromCurve(
     FloatCurve=DataInterface() {propertyValues {Curve {keys [
         {arriveTangent=9.0725, interpMode=RCIM_Cubic, leaveTangent=9.0725, tangentMode=RCTM_User, time=0, value=0}
         {arriveTangent=0.0643, interpMode=RCIM_Cubic, leaveTangent=0.0643, tangentMode=RCTM_User, time=0.1408, value=0.9847}
         {arriveTangent=-1.8333, interpMode=RCIM_Cubic, leaveTangent=-1.8333, tangentMode=RCTM_User, time=0.6, value=0.5}
         {arriveTangent=-0.0405, interpMode=RCIM_Cubic, leaveTangent=-0.0405, tangentMode=RCTM_User, time=1, value=0}
        ]}}}
     Scale Curve=0.2
    )
    Scale RGB=&VectorFromCurve(VectorCurve=DataInterface() {propertyValues {
       XCurve {keys [<at 0 0.2 RCIM_Cubic>,<at 0.2 0.7 RCIM_Cubic>]}
       YCurve {keys [<at 0 4 RCIM_Cubic>,<at 0.2 0.7 RCIM_Cubic>]}
       ZCurve {keys [<at 0 5 RCIM_Cubic>,<at 0.2 0.7 RCIM_Cubic>]}
      }})
    ScaleA=true
    ScaleRGB=false
   )
   &ScaleSpriteSize(Scale Factor=&Vector2DFromFloat(Value=&FloatFromCurve(FloatCurve=DataInterface() {propertyValues {Curve {keys [{arriveTangent=6.9817, interpMode=RCIM_Cubic, time=0, value=1.5}, <at 1 2 RCIM_Cubic>]}}})))
   &SpriteRotationRate(Rotation Rate=&NormalizedAngleToDegrees(NormalizedAngle=$Particles.SpriteRotationRate))
  ]
  rendererClasses [&NiagaraSpriteRendererProperties]
  renderers [{rendererClass=&NiagaraSpriteRendererProperties}]
 }
 SparksStartLeft {
  bEnabled=false
  simTarget=GPUComputeSim
  emitterUpdateScript [
   &EmitterState()
   &SpawnRate(
    SpawnRate=&Multiply_Float(
     A=60
     B=&FloatFromCurve(
      FloatCurve=DataInterface() {propertyValues {Curve {keys [<at 0 1 RCIM_Cubic>, {arriveTangent=-0.45, interpMode=RCIM_Cubic, leaveTangent=-0.45, time=0.3, value=0.95}, <at 2 0.1>]}}}
      CurveIndex=$Emitter.NormalizedLoopAge
     )
    )
    Use Spawn Probability=false
   )
   &SpawnBurst_Instantaneous(Spawn Count=&RandomRangeInt(Minimum=20, Maximum=30), Use Spawn Probability=false)
  ]
  particleSpawnScript [
   &InitializeParticle(
    Color=$User.Color
    Lifetime=&RandomRangeFloat(Minimum=0.08, Maximum=0.15, Evaluation Type=::NewEnumerator1)
    Module.Write Color=true
    Module.Write Lifetime=true
    Module.Write Mass=true
    Module.Write Material Random=true
    Module.Write Position=false
    Module.Write Scale=false
    Module.Write SpriteRotation=false
    Module.Write SpriteSize=true
    Position=&RandomRangeVector(Minimum=<0 75 20>, Maximum=<0 77 20>, Evaluation Type=::NewEnumerator1)
    Sprite Size=&RandomRangeVector2D(Minimum=<0.7 0.5>, Maximum=<1.2 1.5>, Evaluation Type=::NewEnumerator1)
   )
   /Niagara/Modules/Spawn/Location/SphereLocation(
    Hemisphere Y=true
    MaskSpawnGroup=false
    Module.Override Randomness=false
    Sphere Radius=&RandomRangeFloat(Minimum=2, Maximum=7, Evaluation Type=::NewEnumerator1)
    Surface Only=false
   )
   &AddVelocity(Velocity=&Multiply_VectorByFloat(
     Vector=$Output.SphereLocation.SphereVector
     Float=&RandomRangeFloat(Minimum=15, Maximum=35, Evaluation Type=::NewEnumerator1)
    ))
   &AddVelocity(Velocity=&RandomRangeVector(Minimum=<-50 50 -50>, Maximum=<50 0 50>, Evaluation Type=::NewEnumerator1)) {module=AddVelocity001}
  ]
  particleUpdateScript [
   &ParticleState()
   &SolveForcesAndVelocity(Clamp Velocity=false, Limit Acceleration=false)
   &ScaleSpriteSize(Scale Factor=&Vector2DFromFloat(Value=&FloatFromCurve(FloatCurve=DataInterface() {propertyValues {Curve {keys [
          {arriveTangent=55.4408, interpMode=RCIM_Cubic, time=0, value=0}
          <at 0.03 1 RCIM_Cubic>
          {arriveTangent=-3.3237, leaveTangent=-3.3237, time=1, value=0}
         ]}}})))
   &ScaleColor(
    Scale Alpha=&FloatFromCurve(FloatCurve=DataInterface() {propertyValues {Curve {keys [{arriveTangent=-6.8352, interpMode=RCIM_Cubic, time=0, value=2}, <at 0.2 1>]}}})
    Scale RGB=&VectorFromCurve(VectorCurve=DataInterface() {propertyValues {
       XCurve {keys [<at 0 16 RCIM_Cubic>,<at 0.8 1 RCIM_Cubic>]}
       YCurve {keys [<at 0 16 RCIM_Cubic>,<at 0.8 0.1249 RCIM_Cubic>]}
       ZCurve {keys [<at 0 16 RCIM_Cubic>,<at 0.8 0.2547 RCIM_Cubic>]}
      }})
    ScaleA=true
    ScaleRGB=true
   )
   &ScaleSpriteSizeBySpeed(Max Scale Factor=<1 20>, Source Velocity:Vector3f=&Multiply_Vector(A=<0 0.015 1>, B=$Particles.Velocity))
  ]
  rendererClasses [&NiagaraSpriteRendererProperties]
  renderers [{rendererClass=&NiagaraSpriteRendererProperties}]
 }
}
summary {
 systemName=NS_Knockback_Overheat
 emitters [
  {emitterName=Smoke_Barrel, rendererClasses [&NiagaraSpriteRendererProperties]}
  {bEnabled=false, emitterName=Particle_Emitter, rendererClasses []}
  {
   bEnabled=false
   emitterName=Light
   rendererClasses [&NiagaraSpriteRendererProperties, &NiagaraLightRendererProperties]
  }
  {bEnabled=false, emitterName=Particle_Emitter001, rendererClasses []}
  {bEnabled=false, emitterName=BoltsLarge, rendererClasses [&NiagaraLightRendererProperties]}
  {emitterName=BigGlow, rendererClasses [&NiagaraSpriteRendererProperties]}
  {bEnabled=false, emitterName=SparksStartLeft, simTarget=GPUComputeSim, rendererClasses [&NiagaraSpriteRendererProperties]}
  {emitterName=Decal, rendererClasses [&NiagaraSpriteRendererProperties]}
 ]
 userVariables [User.Color {type=LinearColor, description="", defaultValue {struct=LinearColor, value=<rgba 1.5 0.2 0.1 1>}}]
}
